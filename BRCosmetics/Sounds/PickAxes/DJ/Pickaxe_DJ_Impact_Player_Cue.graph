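# Sound cue for the DJ pickaxe cosmetic's impact sound when it hits players.
# A mixer layers a random Douyu pickaxe enemy-impact wave after a short delay with a random DJ player-impact wave that starts slightly later with randomized pitch.
# It uses the pickaxe impact property sheet, player impact sound class and melee weapon concurrency, with overridden logarithmic attenuation that adds distance-based low-pass filtering.
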SoundCue /BRCosmetics/Sounds/PickAxes/DJ/Pickaxe_DJ_Impact_Player_Cue
root=Mixer_0
nodes [
 {id=Delay_0, children [Random_2], properties {delayMax=0.2, delayMin=0.2}}
 {id=Delay_1, children [Random_3], properties {delayMax=0.1, delayMin=0.1}}
 {id=Mixer_0, children [Delay_1,Modulator_0], properties {inputVolume [1,1]}}
 {id=Modulator_0, children [Delay_0], properties {pitchMax=1.25, pitchMin=0.8}}
 {id=Random_2, children [WavePlayer_10, WavePlayer_11, WavePlayer_12, WavePlayer_13], properties {weights [1,1,1,1]}}
 {
  id=Random_3
  children [WavePlayer_14, WavePlayer_15, WavePlayer_16, WavePlayer_17, WavePlayer_18, WavePlayer_19]
  properties {weights [1,1,1,1,1,1]}
 }
 {id=WavePlayer_10, properties {soundWaveAssetPtr=&/Pickaxe_DJ_Impact_Player_03}}
 {id=WavePlayer_11, properties {soundWaveAssetPtr=&/Pickaxe_DJ_Impact_Player_04}}
 {id=WavePlayer_12, properties {soundWaveAssetPtr=&/Pickaxe_DJ_Impact_Player_05}}
 {id=WavePlayer_13, properties {soundWaveAssetPtr=&/Pickaxe_DJ_Impact_Player_06}}
 {id=WavePlayer_14, properties {soundWaveAssetPtr=/BRCosmetics/Sounds/PickAxes/Douyu/PickAxe_Douyu_EnemyImpact_06}}
 {id=WavePlayer_15, properties {soundWaveAssetPtr=/BRCosmetics/Sounds/PickAxes/Douyu/PickAxe_Douyu_EnemyImpact_01}}
 {id=WavePlayer_16, properties {soundWaveAssetPtr=/BRCosmetics/Sounds/PickAxes/Douyu/PickAxe_Douyu_EnemyImpact_02}}
 {id=WavePlayer_17, properties {soundWaveAssetPtr=/BRCosmetics/Sounds/PickAxes/Douyu/PickAxe_Douyu_EnemyImpact_03}}
 {id=WavePlayer_18, properties {soundWaveAssetPtr=/BRCosmetics/Sounds/PickAxes/Douyu/PickAxe_Douyu_EnemyImpact_04}}
 {id=WavePlayer_19, properties {soundWaveAssetPtr=/BRCosmetics/Sounds/PickAxes/Douyu/PickAxe_Douyu_EnemyImpact_05}}
]
properties {
 attenuationSettings=/Game/Sounds/Attenuation/Fort_Player_Footsteps_Att
 audioPropertiesSheet=/BRCosmetics/Sounds/Settings/PropertySheets/Pickaxe/APS_Cosmetic_Pickaxe_Impact
 bOverrideAttenuation=true
 maxDistance=1100
 soundClassObject=/Game/Sounds/FortSoundClassesAndModes/Fort_Impact_Player_PSM
 volumeMultiplier=0.8
 attenuationOverrides {
  bAttenuateWithLPF=true
  distanceAlgorithm=Logarithmic
  falloffDistance=800
  lPFRadiusMax=800
  lPFRadiusMin=400
  attenuationShapeExtents {x=300}
 }
 concurrencyOverrides {bLimitToOwner=true, maxCount=2, resolutionRule=StopQuietest}
 concurrencySet [&Player_Melee_Weapon_Conc]
}
